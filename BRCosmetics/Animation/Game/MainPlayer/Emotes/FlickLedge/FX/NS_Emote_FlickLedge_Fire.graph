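# Niagara system for the fire effect of the FlickLedge emote, attached to the character's skeletal mesh.
# Flame, burst and ember emitters spawn from a shared mesh reader, using spawn groups to pick sockets, while a flame emitter adds flickering lights.
# Curves indexed by emitter loop age choreograph spawn rate, speed and brightness into surges and lulls.
# A separate emitter follows the FlickLedge head bone and shifts from orange to blue mid-loop.
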NiagaraSystem /BRCosmetics/Animation/Game/MainPlayer/Emotes/FlickLedge/FX/NS_Emote_FlickLedge_Fire
dependencies {
 usedDataInterfaces [
  &NiagaraDataInterfaceCurve
  &NiagaraDataInterfaceSpriteRendererInfo
  &NiagaraDataInterfaceColorCurve
  &NiagaraDataInterfaceArrayFloat
  &NiagaraDataInterfaceSkeletalMesh
  &NiagaraDataInterfaceCamera
 ]
 usedDynamicInputs [
  &SelectFloatFromArray
  &Multiply_Vector4
  &MakeVector4_FromLinearColor
 ]
 usedModules [
  &self:SystemSpawnScript.SystemScriptSource.SystemScriptGraph.NiagaraNodeAssignment_1.SetVariables_#1
  &SystemState
  &SampleSkeletalMesh
 ]
}
emitters {
 BurstFlame {
  emitterUpdateScript [
   &EmitterState()
   &SpawnRate(
    SpawnRate=&FloatFromCurve(
     FloatCurve=DataInterface() {propertyValues {Curve {keys [
         <at 0 1>
         <at 0.95 1>
         <at 1 1>
         <at 2 1>
         <at 2.033 1>
         {arriveTangent=-0.0285, interpMode=RCIM_Cubic, leaveTangent=-0.0285, tangentMode=RCTM_User, time=2.6426, value=1}
         {arriveTangent=-1.4254, interpMode=RCIM_Cubic, leaveTangent=-1.4254, tangentMode=RCTM_User, time=4, value=0}
        ]}}}
     CurveIndex=$Emitter.LoopedAge
     Scale Curve=96
    )
    Use Spawn Probability=false
   )
  ]
  particleSpawnScript [
   /SetVariables(Particles.Random size:NiagaraFloat=&RandomRangeFloat(Minimum=0.1, Maximum=0.5)) {isSetParameters=true}
   &InitializeParticle(
    Lifetime Max=0.425
    Lifetime Min=0.3
    Lifetime Mode=::NewEnumerator1
    Sprite Rotation Mode=::NewEnumerator2
    Sprite Size Mode=::NewEnumerator3
    Uniform Sprite Size=&Multiply_Float(
     A=&FloatFromCurve(
      FloatCurve=DataInterface() {propertyValues {Curve {keys [<at 0 1>, {arriveTangent=-0.25, interpMode=RCIM_Cubic, leaveTangent=-0.25, time=3.1, value=1}, <at 4 0 RCIM_Cubic>]}}}
      CurveIndex=$Emitter.LoopedAge
      Scale Curve=56
     )
     B=$Particles.Random size
    )
    UsePositionOffset=false
    Write Material Random=true
   )
   &ShapeLocation(
    Cylinder Height=28
    Cylinder Radius=4
    MaskSpawnGroup=false
    Override Seed=false
    Shape Primitive=::NewEnumerator1
    Surface Only=false
    UseOwnerScale=true
   )
   &SkeletalMeshLocation(
    MaskSpawnGroup=false
    Mesh Sampling Type=::NewEnumerator5
    Sampled Position Offset=&AddVectorToPosition(Position=$Particles.Initial.ShapeLocation.ShapePosition, Vector=<-15 0 0>)
    Skeletal Mesh=$System.Skeletal Mesh Reader
    Socket Sampling Mode=::NewEnumerator5
   )
   &AddVelocity(Velocity=&RandomRangeVector(Minimum=<0 0 45>, Maximum=<0 0 125>))
   &DynamicMaterialParameters(
    Index 0 Param 4=2.5
    Module.Param0WriteEnabled=true
    Module.Param1WriteEnabled=true
    Module.Param2WriteEnabled=false
    Module.Param3WriteEnabled=true
   ) {module=DynamicMaterialParameters001}
  ]
  particleUpdateScript [
   /SetVariables(Particles.Size Scale:NiagaraFloat=&FloatFromCurve(
     FloatCurve=DataInterface() {propertyValues {Curve {keys [
         {interpMode=RCIM_Cubic, tangentMode=RCTM_User, time=0, value=0}
         {interpMode=RCIM_Cubic, tangentMode=RCTM_User, time=0.2, value=0.3}
         {tangentMode=RCTM_User, time=1, value=0.1}
        ]}}}
     Scale Curve=6
    )) {isSetParameters=true}
   /Niagara/Modules/Update/Forces/CurlNoiseForce(Noise Frequency=100, Noise Strength=&Multiply_Float(A=100, B=3), Pan Noise=true)
   &Drag(Drag=&RandomRangeFloat(Minimum=0.8, Maximum=2), Use Linear Drag=true, Use Rotational Drag=true)
   &GravityForce(Gravity=<0 0 100>)
   &SolveForcesAndVelocity(Clamp Velocity=false, Limit Acceleration=false)
   &ParticleState()
   &SpriteRotationRate(Rotation Rate=&RandomRangeFloat(Minimum=-300, Maximum=300)) {module=SpriteRotationRate001}
   &ScaleSpriteSize(Scale Sprite Size Mode=::NewEnumerator0, Uniform Scale Factor=$Particles.Size Scale) {module=ScaleSpriteSize001}
   &ScaleColor(
    Scale Alpha=&Multiply_Float(
     A=4
     B=&FloatFromCurve(
      FloatCurve=DataInterface() {propertyValues {Curve {keys [<at 0 1 RCIM_Cubic>, {interpMode=RCIM_Cubic, tangentMode=RCTM_User, time=0.5, value=1}, {tangentMode=RCTM_User, time=1, value=0.1}]}}}
      Scale Curve=0.7
     )
    )
    Scale RGB=&Multiply_VectorByFloat(
     Vector=&MakeVectorFromLinearColorRGB(Color=&ColorFromCurve(ColorCurve=DataInterface() {propertyValues {
         AlphaCurve {keys [<at 0 1>]}
         BlueCurve {keys [<at 0.3404 0 RCIM_Cubic>,<at 1 0>]}
         GreenCurve {keys [<at 0.3404 0.3271 RCIM_Cubic>,<at 1 0.6659>]}
         RedCurve {keys [<at 0.3404 4 RCIM_Cubic>,<at 1 1>]}
        }}))
     Float=1
    )
    ScaleA=true
    ScaleRGB=true
   )
   &FN_Sprite_Attributes(
    Color=&MakeLinearColorFromVectorAndFloat(
     "Vector (RGB)"=&MakeVectorFromLinearColorRGB(Color=<rgba 1 0.2 0 0.5>)
     "Float (Alpha)"=&FN_FrontEndFloatModifier(Gameplay=0.099, Front End=0.04)
    )
    Create Color Attribute=true
    Create SpriteSize Attribute=true
    SpriteSize=&Vector2DFromFloat(Value=&Multiply_Float(A=120, B=&Multiply_Float(A=$Particles.Random size, B=$Particles.Size Scale)))
    Write to Module Namespace=true
   ) {enabled=false, module=FN_Sprite_Attributes001}
   &SubUVAnimation(
    End Frame Range Override=12
    "Sprite Renderer "=DataInterface() {propertyValues {SpriteRenderer=&self:BurstFlame_3.NiagaraSpriteRendererProperties_0}}
    Start Frame Range Override=8
    SubUV Lookup Index Scale:NiagaraFloat=&RandomRangeFloat(Minimum=0.2, Maximum=1)
    UseEndFrame=true
    UseStartFrame=true
   )
  ]
  rendererClasses [&NiagaraSpriteRendererProperties]
  renderers [
   {rendererClass=&NiagaraSpriteRendererProperties}
   {rendererClass=&NiagaraSpriteRendererProperties, rendererIndex=1}
  ]
 }
 Empty {rendererClasses [], renderers []}
 Flames003 {
  emitterUpdateScript [
   &EmitterState()
   &SpawnBurst_Instantaneous(Spawn Count=12, Use Spawn Probability=false, UseLoopCountLimit=false)
   &SpawnRate(
    SpawnRate=&FloatFromCurve(
     FloatCurve=DataInterface() {propertyValues {Curve {keys [<at 0 1>,<at 1.05 1>,<at 1.083 0.5>,<at 2 0.5>,<at 2.033 1>,<at 2.6426 1>,<at 3.1941 1.0194>,<at 3.25 0>]}}}
     CurveIndex=$Emitter.LoopedAge
     Scale Curve=10
    )
    Use Spawn Probability=false
   )
   &SpawnRate(
    SpawnGroup=1
    SpawnRate=&FloatFromCurve(
     FloatCurve=DataInterface() {propertyValues {Curve {keys [<at 0 1>,<at 1.05 1>,<at 1.083 0.5>,<at 2 0.5>,<at 2.033 1>,<at 2.6426 1>,<at 3.1941 1.0194>,<at 3.25 0>]}}}
     CurveIndex=$Emitter.LoopedAge
     Scale Curve=8
    )
    Use Spawn Probability=false
   ) {module=SpawnRate001}
   &SpawnRate(
    SpawnGroup=2
    SpawnRate=&FloatFromCurve(
     FloatCurve=DataInterface() {propertyValues {Curve {keys [<at 0 1>,<at 1.05 1>,<at 1.083 0.5>,<at 2 0.5>,<at 2.033 1>,<at 2.6426 1>,<at 3.1941 1.0194>,<at 3.25 0>]}}}
     CurveIndex=$Emitter.LoopedAge
     Scale Curve=8
    )
    Use Spawn Probability=false
   ) {module=SpawnRate002}
  ]
  particleSpawnScript [
   &InitializeParticle(
    Lifetime=&RandomRangeFloat(Minimum=0.475, Maximum=0.55)
    Sprite Rotation Mode=::NewEnumerator2
    Sprite Size Mode=::NewEnumerator3
    Uniform Sprite Size=&MergeHelper_FloatSwitch(
     0=&FloatFromCurve(
      FloatCurve=DataInterface() {propertyValues {Curve {keys [<at 0 1>,<at 3.5 1>,<at 3.7 0.475>]}}}
      CurveIndex=$Emitter.LoopedAge
      Scale Curve=&RandomRangeFloat(Minimum=24, Maximum=28)
     )
     1=&RandomRangeFloat(Minimum=18, Maximum=22)
     2=&RandomRangeFloat(Minimum=18, Maximum=22)
     3=&RandomRangeFloat(Minimum=20, Maximum=35)
     Index=&SpawnGroup()
     Num=::NewEnumerator1
    )
    UsePositionOffset=false
    Write Material Random=true
   )
   &ShapeLocation(
    MaskSpawnGroup=false
    Override Seed=false
    Sphere Radius=&MergeHelper_FloatSwitch(0=2, 1=4, 2=4, 3=?, Index=&SpawnGroup(), Num=::NewEnumerator1)
    UseOwnerScale=true
   )
   &SkeletalMeshLocation(
    Bone / Socket Index=&SpawnGroup()
    MaskSpawnGroup=false
    Mesh Sampling Type=::NewEnumerator5
    Sampled Position Offset=&AddVectorToPosition(
     Position=$Particles.ShapeLocation.ShapePosition
     Vector=&MergeHelper_VectorSwitch(
      0=&MakeVector(
       X=0
       Y=0
       Z=&FloatFromCurve(FloatCurve=DataInterface() {propertyValues {Curve {keys [<at 0 -8>,<at 3.5 -8>,<at 3.7 0>]}}}, CurveIndex=$Emitter.LoopedAge)
      )
      1=<-4 0 -4>
      2=<4 0 -4>
      3=?
      Index=&SpawnGroup()
      Num=::NewEnumerator1
     )
    )
    Sampled Position Offset Coordinate Space:ENiagaraSkelMeshTransforms=::NewEnumerator2
    Skeletal Mesh=$System.Skeletal Mesh Reader
    Socket Sampling Mode=::NewEnumerator5
   )
   &AddVelocity(
    Cone Angle=15
    Cone Axis=&MakeVector(
     X=&RandomRangeFloat(Minimum=-0.1, Maximum=0.1)
     Y=&FloatFromCurve(FloatCurve=DataInterface() {propertyValues {Curve {keys [<at 0 1>,<at 0.1 0>]}}}, CurveIndex=$Emitter.NormalizedLoopAge, Scale Curve=-1)
     Z=1
    )
    Module.UseConeFalloff=false
    Module.UseVelDistribution=false
    Override Seed=false
    Velocity Mode=::NewEnumerator2
    Velocity Speed=&MergeHelper_FloatSwitch(
     0=&FloatFromCurve(
      FloatCurve=DataInterface() {propertyValues {Curve {keys [<at 0 1>,<at 1.05 1>,<at 1.083 0.4>,<at 2 0.4>,<at 2.033 1.25>,<at 2.6426 1.25>,<at 3.1941 0.8>]}}}
      CurveIndex=$Emitter.LoopedAge
      Scale Curve=&RandomRangeFloat(Minimum=450, Maximum=480)
     )
     1=&FloatFromCurve(
      FloatCurve=DataInterface() {propertyValues {Curve {keys [<at 0 1>,<at 1.05 1>,<at 1.083 0.4>,<at 2 0.4>,<at 2.033 1.25>,<at 2.6426 1.25>,<at 3.1941 1>]}}}
      CurveIndex=$Emitter.LoopedAge
      Scale Curve=&RandomRangeFloat(Minimum=325, Maximum=400)
     )
     2=&FloatFromCurve(
      FloatCurve=DataInterface() {propertyValues {Curve {keys [<at 0 1>,<at 1.05 1>,<at 1.083 0.4>,<at 2 0.4>,<at 2.033 1.25>,<at 2.6426 1.25>,<at 3.1941 1>]}}}
      CurveIndex=$Emitter.LoopedAge
      Scale Curve=&RandomRangeFloat(Minimum=325, Maximum=400)
     )
     3=?
     Index=&SpawnGroup()
     Num=::NewEnumerator1
    )
   )
   &CameraOffset(Camera Offset Amount=&RandomRangeFloat(Minimum=0, Maximum=5))
  ]
  particleUpdateScript [
   &ParticleState()
   &ScaleSpriteSize(Uniform Curve Sprite Scale=DataInterface() {propertyValues {Curve {keys [
        <at 0 0.6 RCIM_Cubic>
        {arriveTangent=1.0711, interpMode=RCIM_Cubic, leaveTangent=1.0711, time=0.15, value=1}
        {arriveTangent=-0.6667, interpMode=RCIM_Cubic, leaveTangent=-0.6667, time=0.3662, value=0.9922}
        {arriveTangent=-1.5655, interpMode=RCIM_Cubic, leaveTangent=-1.5655, time=0.6, value=0.7}
        <at 1 0 RCIM_Cubic>
       ]}}})
   &Drag(
    Drag=&MergeHelper_FloatSwitch(
     0=&FloatFromCurve(
      FloatCurve=DataInterface() {propertyValues {Curve {keys [<at 0 1>, {arriveTangent=-0.2857, interpMode=RCIM_Cubic, leaveTangent=-0.2857, time=0.6, value=1}, <at 0.7 0.8 RCIM_Cubic>]}}}
      CurveIndex=$Emitter.NormalizedLoopAge
      Scale Curve=&RandomRangeFloat(Minimum=6, Maximum=8)
     )
     1=&RandomRangeFloat(Minimum=5, Maximum=8)
     2=&RandomRangeFloat(Minimum=5, Maximum=8)
     3=?
     Index=&SpawnGroup()
     Num=::NewEnumerator1
    )
    Use Linear Drag=true
    Use Rotational Drag=false
   )
   &SubUVAnimation(End Frame=16)
   &SolveForcesAndVelocity(Clamp Velocity=false, Limit Acceleration=false)
   &Color(
    Color=&Multiply_LinearColorByFloat(
     Linear Color=&ColorFromCurve(ColorCurve=DataInterface() {propertyValues {
        AlphaCurve {keys [<at 0 1>]}
        BlueCurve {keys [<at 0 0 RCIM_Cubic>]}
        GreenCurve {keys [<at 0 0.6563 RCIM_Cubic>]}
        RedCurve {keys [<at 0 4 RCIM_Cubic>]}
       }})
     Float=&FloatFromCurve(
      FloatCurve=DataInterface() {propertyValues {Curve {defaultValue=1, keys [<at 0 1.4>,<at 1.05 1.4>,<at 1.083 1>,<at 2 1>,<at 2.033 1.4>,<at 2.6426 1.4>,<at 3.1941 1.0194>,<at 5.75 1>,<at 6 0>]}}}
      CurveIndex=$Emitter.LoopedAge
      Scale Curve=0.45
     )
    )
    Scale Alpha=&FloatFromCurve(FloatCurve=DataInterface() {propertyValues {Curve {keys [<at 0.5 1>,<at 1 0>]}}}, Scale Curve=0.2)
   )
  ]
  rendererClasses [&NiagaraSpriteRendererProperties]
  renderers [{rendererClass=&NiagaraSpriteRendererProperties}]
 }
 Flames004 {
  emitterUpdateScript [
   &EmitterState()
   &SpawnBurst_Instantaneous(Spawn Count=18, Use Spawn Probability=false, UseLoopCountLimit=false)
   &SpawnRate(
    SpawnRate=&FloatFromCurve(
     FloatCurve=DataInterface() {propertyValues {Curve {keys [<at 0 1>,<at 1.05 1>,<at 1.3 0>,<at 1.7 0>,<at 2.033 1>,<at 2.6426 1>,<at 3.1941 1.0194>,<at 3.25 0>]}}}
     CurveIndex=$Emitter.LoopedAge
     Scale Curve=24
    )
    Use Spawn Probability=false
   )
   &SpawnRate(
    SpawnGroup=1
    SpawnRate=&FloatFromCurve(
     FloatCurve=DataInterface() {propertyValues {Curve {keys [<at 0 1>,<at 1.05 1>,<at 1.3 0>,<at 1.7 0>,<at 2.033 1>,<at 3 1>,<at 3.183 0>]}}}
     CurveIndex=$Emitter.LoopedAge
     Scale Curve=14
    )
    Use Spawn Probability=false
   ) {module=SpawnRate001}
   &SpawnRate(
    SpawnGroup=2
    SpawnRate=&FloatFromCurve(
     FloatCurve=DataInterface() {propertyValues {Curve {keys [<at 0 1>,<at 1.05 1>,<at 1.3 0>,<at 1.7 0>,<at 2.033 1>,<at 3 1>,<at 3.183 0>]}}}
     CurveIndex=$Emitter.LoopedAge
     Scale Curve=14
    )
    Use Spawn Probability=false
   ) {module=SpawnRate002}
  ]
  particleSpawnScript [
   &InitializeParticle(
    Lifetime=&RandomRangeFloat(Minimum=0.425, Maximum=0.5)
    Sprite Rotation Mode=::NewEnumerator2
    Sprite Size Mode=::NewEnumerator3
    Uniform Sprite Size=&MergeHelper_FloatSwitch(
     0=&RandomRangeFloat(Minimum=60, Maximum=70)
     1=&RandomRangeFloat(Minimum=55, Maximum=60)
     2=&RandomRangeFloat(Minimum=55, Maximum=60)
     3=&RandomRangeFloat(Minimum=20, Maximum=35)
     Index=&SpawnGroup()
     Num=::NewEnumerator1
    )
    UsePositionOffset=false
    Write Material Random=true
   )
   &ShapeLocation(
    MaskSpawnGroup=false
    Override Seed=false
    Sphere Radius=&MergeHelper_FloatSwitch(0=2, 1=4, 2=4, 3=?, Index=&SpawnGroup(), Num=::NewEnumerator1)
    UseOwnerScale=true
   )
   &SkeletalMeshLocation(
    Bone / Socket Index=&SpawnGroup()
    MaskSpawnGroup=false
    Mesh Sampling Type=::NewEnumerator5
    Position Sampling=::NewEnumerator2
    Sampled Position Offset Coordinate Space:ENiagaraSkelMeshTransforms=::NewEnumerator2
    Skeletal Mesh=$System.Skeletal Mesh Reader
    Socket Sampling Mode=::NewEnumerator5
   ) {enabled=false}
   &SkeletalMeshLocation(
    Bone / Socket Index=&SpawnGroup()
    MaskSpawnGroup=false
    Mesh Sampling Type=::NewEnumerator5
    Sampled Position Offset=&AddVectorToPosition(
     Position=$Particles.ShapeLocation.ShapePosition
     Vector=&MergeHelper_VectorSwitch(
      0=&MakeVector(
       X=0
       Y=0
       Z=&FloatFromCurve(FloatCurve=DataInterface() {propertyValues {Curve {keys [<at 0 -8>,<at 3.5 -8>,<at 3.7 8>]}}}, CurveIndex=$Emitter.LoopedAge)
      )
      1=<-4 0 -4>
      2=<4 0 -4>
      3=?
      Index=&SpawnGroup()
      Num=::NewEnumerator1
     )
    )
    Sampled Position Offset Coordinate Space:ENiagaraSkelMeshTransforms=::NewEnumerator2
    Skeletal Mesh=$System.Skeletal Mesh Reader
    Socket Sampling Mode=::NewEnumerator5
   ) {module=SkeletalMeshLocation001}
   &AddVelocity(
    Cone Angle=15
    Cone Axis=&MergeHelper_VectorSwitch(
     0=&ReturnMeshOrientationAxis(Mesh Orientation Axis=::XAxis, Mesh Orientation Quat=$Particles.MeshOrientation)
     1=&MakeVector(
      X=&RandomRangeFloat(Minimum=-0.1, Maximum=0.1)
      Y=&FloatFromCurve(FloatCurve=DataInterface() {propertyValues {Curve {keys [<at 0 1>,<at 0.1 0>]}}}, CurveIndex=$Emitter.NormalizedLoopAge, Scale Curve=-1)
      Z=1
     )
     2=&MakeVector(
      X=&RandomRangeFloat(Minimum=-0.1, Maximum=0.1)
      Y=&FloatFromCurve(FloatCurve=DataInterface() {propertyValues {Curve {keys [<at 0 1>,<at 0.1 0>]}}}, CurveIndex=$Emitter.NormalizedLoopAge, Scale Curve=-1)
      Z=1
     )
     3=?
     Index=&SpawnGroup()
     Num=::NewEnumerator1
    )
    Module.UseConeFalloff=false
    Module.UseVelDistribution=false
    Override Seed=false
    Velocity Mode=::NewEnumerator2
    Velocity Speed=&MergeHelper_FloatSwitch(
     0=&FloatFromCurve(
      FloatCurve=DataInterface() {propertyValues {Curve {keys [<at 0 1>,<at 1.05 1>,<at 1.083 0.4>,<at 2 0.4>,<at 2.033 1.25>,<at 2.6426 1.25>,<at 3.1941 0.8>]}}}
      CurveIndex=$Emitter.LoopedAge
      Scale Curve=&RandomRangeFloat(Minimum=325, Maximum=500)
     )
     1=&FloatFromCurve(
      FloatCurve=DataInterface() {propertyValues {Curve {keys [<at 0 1>,<at 1.05 1>,<at 1.083 0.4>,<at 2 0.4>,<at 2.033 1.25>,<at 2.6426 1.25>,<at 3.1941 1>]}}}
      CurveIndex=$Emitter.LoopedAge
      Scale Curve=&RandomRangeFloat(Minimum=280, Maximum=300)
     )
     2=&FloatFromCurve(
      FloatCurve=DataInterface() {propertyValues {Curve {keys [<at 0 1>,<at 1.05 1>,<at 1.083 0.4>,<at 2 0.4>,<at 2.033 1.25>,<at 2.6426 1.25>,<at 3.1941 1>]}}}
      CurveIndex=$Emitter.LoopedAge
      Scale Curve=&RandomRangeFloat(Minimum=280, Maximum=300)
     )
     3=?
     Index=&SpawnGroup()
     Num=::NewEnumerator1
    )
   )
   &SubUVAnimation(End Frame=3, SubUV Animation Mode:ENiagara_SubUVLookupMode=::NewEnumerator2)
   &CameraOffset(Camera Offset Amount=&RandomRangeFloat(Minimum=0, Maximum=5))
  ]
  particleUpdateScript [
   &ParticleState()
   &SpriteRotationRate(Rotation Rate=&UniformAOrBFloat(A=-250, B=250, Evaluation Type=::NewEnumerator0, Recalculate Random Each Loop=true, Override Seed=false))
   &ScaleSpriteSize(Uniform Curve Sprite Scale=DataInterface() {propertyValues {Curve {keys [
        {arriveTangent=3.157, interpMode=RCIM_Cubic, leaveTangent=3.157, tangentMode=RCTM_User, time=0, value=0.6}
        {arriveTangent=0.2096, interpMode=RCIM_Cubic, leaveTangent=0.2096, tangentMode=RCTM_User, time=0.15, value=1}
        {arriveTangent=-1.7, interpMode=RCIM_Cubic, leaveTangent=-1.7, tangentMode=RCTM_User, time=0.4639, value=0.4687}
        {arriveTangent=-0.0436, interpMode=RCIM_Cubic, leaveTangent=-0.0436, tangentMode=RCTM_User, time=1, value=0}
       ]}}})
   &Drag(
    Drag=&MergeHelper_FloatSwitch(
     0=&FloatFromCurve(
      FloatCurve=DataInterface() {propertyValues {Curve {keys [<at 0 1>, {arriveTangent=-0.2857, interpMode=RCIM_Cubic, leaveTangent=-0.2857, time=0.6, value=1}, <at 0.7 0.8 RCIM_Cubic>]}}}
      CurveIndex=$Emitter.NormalizedLoopAge
      Scale Curve=&UniformAOrBFloat(A=5.25, B=8, Distribution Weight=0.9, Evaluation Type=::NewEnumerator0, Recalculate Random Each Loop=true, Override Seed=false)
     )
     1=&RandomRangeFloat(Minimum=5, Maximum=7)
     2=&RandomRangeFloat(Minimum=5, Maximum=7)
     3=?
     Index=&SpawnGroup()
     Num=::NewEnumerator1
    )
    Use Linear Drag=true
    Use Rotational Drag=false
   )
   &SolveForcesAndVelocity(Clamp Velocity=false, Limit Acceleration=false)
   &Color(
    Color=&Multiply_LinearColorByFloat(
     Linear Color=&ColorFromCurve(ColorCurve=DataInterface() {propertyValues {
        AlphaCurve {keys [<at 0 1>]}
        BlueCurve {keys [<at 0.3797 0 RCIM_Cubic>,<at 0.566 0>]}
        GreenCurve {keys [<at 0.3797 0.6921 RCIM_Cubic>,<at 0.566 0>]}
        RedCurve {keys [<at 0.3797 12 RCIM_Cubic>,<at 0.566 4>]}
       }})
     Float=&FloatFromCurve(
      FloatCurve=DataInterface() {propertyValues {Curve {defaultValue=1, keys [<at 0 1.4>,<at 1.05 1.4>,<at 1.083 1>,<at 2 1>,<at 2.033 1.4>,<at 2.6426 1.4>,<at 3.1941 1.0194>,<at 5.75 1>,<at 6 0>]}}}
      CurveIndex=$Emitter.LoopedAge
     )
    )
    Scale Alpha=&FloatFromCurve(FloatCurve=DataInterface() {propertyValues {Curve {keys [<at 0.5 1>,<at 1 0>]}}}, Scale Curve=0.2)
   )
   &Color(
    Color=&Multiply_LinearColorByFloat(
     Linear Color=&ColorFromCurve(ColorCurve=DataInterface() {propertyValues {
        AlphaCurve {keys [<at 0 1>]}
        BlueCurve {keys [<at 0.3797 0 RCIM_Cubic>,<at 0.566 0>]}
        GreenCurve {keys [<at 0.3797 1.8705 RCIM_Cubic>,<at 0.566 0.2275>]}
        RedCurve {keys [<at 0.3797 12 RCIM_Cubic>,<at 0.566 4>]}
       }})
     Float=&FloatFromCurve(
      FloatCurve=DataInterface() {propertyValues {Curve {defaultValue=1, keys [<at 0 1.4>,<at 1.05 1.4>,<at 1.083 1>,<at 2 1>,<at 2.033 1.4>,<at 2.6426 1.4>,<at 3.1941 1.0194>,<at 5.75 1>,<at 6 0>]}}}
      CurveIndex=$Emitter.LoopedAge
     )
    )
    Scale Alpha=&FloatFromCurve(FloatCurve=DataInterface() {propertyValues {Curve {keys [<at 0.5 1>,<at 1 0>]}}}, Scale Curve=0.2)
   ) {enabled=false, module=Color001}
   /SetVariables(
    Particles.Glow Color=&Multiply_LinearColorByFloat(
     Linear Color=$Particles.Color
     Float=&FloatFromCurve(
      FloatCurve=DataInterface() {propertyValues {Curve {
         defaultValue=1
         keys [
          {arriveTangent=9.0725, interpMode=RCIM_Cubic, leaveTangent=9.0725, tangentMode=RCTM_User, time=0, value=0}
          {arriveTangent=-2.5249, interpMode=RCIM_Cubic, leaveTangent=-2.5249, tangentMode=RCTM_User, time=0.1408, value=0.9847}
          {arriveTangent=-0.0405, interpMode=RCIM_Cubic, leaveTangent=-0.0405, tangentMode=RCTM_User, time=1, value=0}
         ]
        }}}
      Scale Curve=0.45
     )
    )
    Particles.Glow Size=&Vector2DFromFloat(Value=&Multiply_Float(
      A=&FloatFromCurve(
       FloatCurve=DataInterface() {propertyValues {Curve {
          defaultValue=0
          keys [
           {arriveTangent=3.1867, interpMode=RCIM_Cubic, leaveTangent=3.1867, tangentMode=RCTM_User, time=0, value=0.6}
           {arriveTangent=0.5707, interpMode=RCIM_Cubic, leaveTangent=0.5707, time=0.2, value=1}
           {arriveTangent=-0.0405, interpMode=RCIM_Cubic, leaveTangent=-0.0405, tangentMode=RCTM_User, time=0.7009, value=1}
           {arriveTangent=-8.0089, interpMode=RCIM_Cubic, leaveTangent=-8.0089, tangentMode=RCTM_User, time=1, value=0}
          ]
         }}}
       Scale Curve=&RandomRangeFloat(Minimum=25, Maximum=32)
      )
      B=&FloatFromCurve(
       FloatCurve=DataInterface() {propertyValues {Curve {
          defaultValue=0
          keys [
           {arriveTangent=0.5707, interpMode=RCIM_Cubic, time=3, value=1}
           {arriveTangent=-0.0405, interpMode=RCIM_Cubic, leaveTangent=-0.0405, tangentMode=RCTM_User, time=3.5, value=0.6}
           {arriveTangent=0.0094, interpMode=RCIM_Cubic, leaveTangent=0.0094, tangentMode=RCTM_User, time=5, value=0.6}
           <at 5.5 0 RCIM_Cubic>
          ]
         }}}
       CurveIndex=$Emitter.LoopedAge
      )
     ))
   ) {isSetParameters=true}
   &DynamicMaterialParameters(
    Index 0 Param 1=&FloatFromCurve(FloatCurve=DataInterface() {propertyValues {Curve {keys [<at 0 -0.7778 RCIM_Cubic>,<at 1 -7 RCIM_Cubic>]}}})
    Module.Index0Float0Write=true
    Module.Index0Float1Write=true
    Module.Index0Float2Write=true
    Module.Index0Float3Write=true
   )
  ]
  rendererClasses [&NiagaraSpriteRendererProperties]
  renderers [
   {rendererClass=&NiagaraSpriteRendererProperties}
   {rendererClass=&NiagaraSpriteRendererProperties, rendererIndex=1}
  ]
 }
 Flames005 {
  emitterUpdateScript [
   &EmitterState()
   &SpawnRate(
    SpawnRate=&FloatFromCurve(
     FloatCurve=DataInterface() {propertyValues {Curve {keys [<at 0 1>,<at 0.95 1>,<at 1 0>,<at 2 0>,<at 2.033 1>,<at 2.6426 1>,<at 3.2 0>]}}}
     CurveIndex=$Emitter.LoopedAge
     Scale Curve=18
    )
    Use Spawn Probability=false
   )
   &SpawnBurst_Instantaneous(Spawn Group=1, Use Spawn Probability=false, UseLoopCountLimit=false)
   &SpawnBurst_Instantaneous(Spawn Group=2, Use Spawn Probability=false, UseLoopCountLimit=false) {module=SpawnBurst_Instantaneous001}
  ]
  particleSpawnScript [
   /SetVariables(Particles.VisibilityTag=&SpawnGroup()) {isSetParameters=true}
   &InitializeParticle(
    Lifetime=&MaskFloatBySpawnGroup(Masked Float=&RandomRangeFloat(Minimum=0.3, Maximum=0.525), Passthrough Float=3.6)
    Sprite Rotation Mode=::NewEnumerator2
    Sprite Size Mode=::NewEnumerator3
    Uniform Sprite Size=&RandomRangeFloat(Minimum=16, Maximum=28)
    UsePositionOffset=false
    Write Material Random=true
   )
   &ShapeLocation(MaskSpawnGroup=false, Override Seed=false, Sphere Radius=12, UseOwnerScale=true)
   &SkeletalMeshLocation(
    Bone / Socket Index=&SpawnGroup()
    MaskSpawnGroup=false
    Mesh Sampling Type=::NewEnumerator5
    Sampled Position Offset=&AddVectorToPosition(Position=$Particles.ShapeLocation.ShapePosition, Vector=<0 -2 6>)
    Sampled Position Offset Coordinate Space:ENiagaraSkelMeshTransforms=::NewEnumerator2
    Skeletal Mesh=$System.Skeletal Mesh Reader
    Socket Sampling Mode=::NewEnumerator5
   )
   &AddVelocity(
    Cone Angle=40
    Cone Axis=&ReturnMeshOrientationAxis(Mesh Orientation Axis=::XAxis, Mesh Orientation Quat=$Particles.MeshOrientation)
    Module.UseConeFalloff=false
    Module.UseVelDistribution=false
    Override Seed=false
    Velocity Mode=::NewEnumerator2
    Velocity Speed=&FloatFromCurve(
     FloatCurve=DataInterface() {propertyValues {Curve {keys [<at 0 1>,<at 1.05 1>,<at 1.083 0.4>,<at 2 0.4>,<at 2.033 1.2>,<at 2.6426 1.2>,<at 3.1941 1.2>,<at 3.5 0.4>,<at 5.75 0.4>,<at 6 0>]}}}
     CurveIndex=$Emitter.LoopedAge
     Scale Curve=&RandomRangeFloat(Minimum=180, Maximum=325)
    )
   )
   &CameraOffset(Camera Offset Amount=&RandomRangeFloat(Minimum=-15, Maximum=0))
  ]
  particleUpdateScript [
   &ParticleState()
   &ScaleSpriteSize(Uniform Curve Sprite Scale=DataInterface() {propertyValues {Curve {
       defaultValue=0
       keys [{interpMode=RCIM_Cubic, tangentMode=RCTM_User, time=0, value=0}, {interpMode=RCIM_Cubic, tangentMode=RCTM_User, time=0.2626, value=0.9806}]
      }}})
   &Drag(Drag=&RandomRangeFloat(Minimum=2, Maximum=3.2), Use Linear Drag=true, Use Rotational Drag=false)
   &SubUVAnimation(End Frame=16)
   &SolveForcesAndVelocity(Clamp Velocity=false, Limit Acceleration=false)
   &ScaleSpriteSizeBySpeed(Max Scale Factor=<0.5 2.2>, Velocity Threshold=100)
   &Color(
    Color=&Multiply_LinearColorByFloat(
     Linear Color=&ColorFromCurve(ColorCurve=DataInterface() {propertyValues {
        AlphaCurve {keys [<at 0 0.45>]}
        BlueCurve {keys [<at 0 0 RCIM_Cubic>]}
        GreenCurve {keys [<at 0 0.6563 RCIM_Cubic>]}
        RedCurve {keys [<at 0 4 RCIM_Cubic>]}
       }})
     Float=&FloatFromCurve(
      FloatCurve=DataInterface() {propertyValues {Curve {defaultValue=1, keys [<at 0 1.4>,<at 1.05 1.4>,<at 1.083 1>,<at 2 1>,<at 2.033 1.4>,<at 2.6426 1.4>,<at 3.1941 1.0194>,<at 5.75 1>,<at 6 0>]}}}
      CurveIndex=$Emitter.LoopedAge
      Scale Curve=&RandomRangeFloat(Minimum=0.8, Maximum=0.625)
     )
    )
    Scale Alpha=&FloatFromCurve(FloatCurve=DataInterface() {propertyValues {Curve {keys [<at 0.5 1>,<at 1 0>]}}}, Scale Curve=0.2)
   )
   &SkeletalMeshLocation(
    MaskSpawnGroup=false
    Mesh Sampling Type=::NewEnumerator5
    Orientation Sampling=::NewEnumerator1
    Position Sampling=::NewEnumerator2
    Skeletal Mesh=$System.Skeletal Mesh Reader
    Socket Sampling Mode=::NewEnumerator5
   ) {module=SkeletalMeshLocation001}
   &FN_Light_Attributes(
    Color=<rgba 1.5 0.4443 0 1>
    Create Color Attribute=true
    Create Exponent Attribute=true
    Create Position Attribute=true
    Create Radius Attribute=true
    Exponent=1
    Position=<0 0 15>
    Radius=&Waveform("Waveform [1]"=::NewEnumerator2, "[1] Frequency"=3, "[1] Amplitude Min/Max"=<80 100>)
    Scale Color=&FloatFromCurve(
     FloatCurve=DataInterface() {propertyValues {Curve {keys [
         <at 0 0 RCIM_Cubic>
         {arriveTangent=0.147, interpMode=RCIM_Cubic, leaveTangent=0.147, tangentMode=RCTM_User, time=0.125, value=1}
         {arriveTangent=0.0765, interpMode=RCIM_Cubic, leaveTangent=0.0765, tangentMode=RCTM_User, time=0.8, value=1}
         <at 1 0 RCIM_Cubic>
        ]}}}
     Scale Curve=&Waveform("Waveform [1]"=::NewEnumerator2, "[1] Frequency"=3, "[1] Amplitude Min/Max"=<0.3 1>)
    )
   )
   &FN_Light_Attributes(
    Color=<rgba 1 0.2962 0 1>
    Create Color Attribute=true
    Create Exponent Attribute=true
    Create Position Attribute=true
    Create Radius Attribute=true
    Exponent=4
    Position=&AddVectorToPosition(Position=$Particles.SkeletalMeshLocation001.SampledPosition, Vector=<0 0 0>)
    Radius=120
    Scale Color=&FloatFromCurve(
     FloatCurve=DataInterface() {propertyValues {Curve {keys [
         <at 0 0 RCIM_Cubic>
         {arriveTangent=0.147, interpMode=RCIM_Cubic, leaveTangent=0.147, tangentMode=RCTM_User, time=0.125, value=1}
         {arriveTangent=0.0765, interpMode=RCIM_Cubic, leaveTangent=0.0765, tangentMode=RCTM_User, time=0.8, value=1}
         <at 1 0 RCIM_Cubic>
        ]}}}
     Scale Curve=&Waveform("Waveform [1]"=::NewEnumerator2, "[1] Frequency"=3, "[1] Compound Frequency"=5, "[1] Amplitude Min/Max"=<0.625 1>, "[1] Scale"=25)
    )
   ) {module=FN_Light_Attributes001}
  ]
  rendererClasses [&NiagaraSpriteRendererProperties, &NiagaraLightRendererProperties]
  renderers [
   {rendererClass=&NiagaraSpriteRendererProperties}
   {rendererClass=&NiagaraLightRendererProperties, rendererIndex=1}
   {rendererClass=&NiagaraLightRendererProperties, rendererIndex=2}
  ]
 }
 NE_Embers001 {
  emitterUpdateScript [
   &EmitterState()
   &SpawnBurst_Instantaneous(Spawn Count=8, Use Spawn Probability=false, UseLoopCountLimit=false)
   &SpawnRate(
    SpawnRate=&FloatFromCurve(
     FloatCurve=DataInterface() {propertyValues {Curve {keys [<at 0 1>,<at 1.05 1>,<at 1.083 0.85>,<at 2 0.85>,<at 2.033 3>,<at 2.6426 3>,<at 3.1941 0>,<at 5.75 0>,<at 6 0>]}}}
     CurveIndex=$Emitter.LoopedAge
     Scale Curve=12
    )
    Use Spawn Probability=true
   )
   &SpawnRate(
    SpawnGroup=1
    SpawnRate=&FloatFromCurve(
     FloatCurve=DataInterface() {propertyValues {Curve {keys [<at 0 1>,<at 1.05 1>,<at 1.083 0.6>,<at 2 0.6>,<at 2.033 1.6>,<at 3 1.6>,<at 3.183 0>]}}}
     CurveIndex=$Emitter.LoopedAge
     Scale Curve=12
    )
    Use Spawn Probability=true
   ) {module=SpawnRate001}
   &SpawnRate(
    SpawnGroup=2
    SpawnRate=&FloatFromCurve(
     FloatCurve=DataInterface() {propertyValues {Curve {keys [<at 0 1>,<at 1.05 1>,<at 1.083 0.6>,<at 2 0.6>,<at 2.033 1.6>,<at 3 1.6>,<at 3.183 0>]}}}
     CurveIndex=$Emitter.LoopedAge
     Scale Curve=12
    )
    Use Spawn Probability=true
   ) {module=SpawnRate002}
  ]
  particleSpawnScript [
   &InitializeParticle(
    Color=<rgba 2 0.7121 0 0.9014>
    Lifetime Max=0.5
    Lifetime Min=0.125
    Lifetime Mode=::NewEnumerator1
    Position Offset=<0 -17.5 -5>
    Sprite Rotation Mode=::NewEnumerator2
    Sprite Size Mode=::NewEnumerator1
    Uniform Sprite Size Max=7
    Uniform Sprite Size Min=2
    UsePositionOffset=false
    Write Material Random=true
   )
   &ShapeLocation(MaskSpawnGroup=false, Override Seed=false, Sphere Radius=10, UseOwnerScale=true)
   &SkeletalMeshLocation(
    Bone / Socket Index=&SpawnGroup()
    MaskSpawnGroup=false
    Mesh Sampling Type=::NewEnumerator5
    Sampled Position Offset=&AddVectorToPosition(Position=$Particles.ShapeLocation.ShapePosition, Vector=<0 0 -5>)
    Skeletal Mesh=$System.Skeletal Mesh Reader
    Socket Sampling Mode=::NewEnumerator5
   )
   &AddVelocity(
    Cone Angle=0
    Cone Axis=<0 0 1>
    Module.UseConeFalloff=false
    Module.UseVelDistribution=false
    Override Seed=false
    Velocity Mode=::NewEnumerator2
    Velocity Speed=&FloatFromCurve(
     FloatCurve=DataInterface() {propertyValues {Curve {keys [<at 0 1>,<at 1.05 1>,<at 1.083 0.6>,<at 2 0.6>,<at 2.033 1.2>,<at 3 1.2>,<at 3.183 0>]}}}
     CurveIndex=$Emitter.LoopedAge
     Scale Curve=&RandomRangeFloat(Minimum=425, Maximum=500)
    )
   )
  ]
  particleUpdateScript [
   &ParticleState()
   &SpriteRotationRate(Rotation Rate=50)
   &ScaleSpriteSize(Uniform Curve Sprite Scale=DataInterface() {propertyValues {Curve {
       defaultValue=0
       keys [
        {arriveTangent=4.8998, interpMode=RCIM_Cubic, leaveTangent=4.8998, tangentMode=RCTM_User, time=0.0003, value=0.0536}
        {interpMode=RCIM_Cubic, tangentMode=RCTM_User, time=0.2908, value=0.9741}
        {arriveTangent=-1.2981, interpMode=RCIM_Cubic, leaveTangent=-1.2981, tangentMode=RCTM_User, time=1, value=0}
       ]
      }}})
   &ScaleColor(
    Scale Alpha=&FloatFromCurve(FloatCurve=DataInterface() {propertyValues {Curve {
        defaultValue=1
        keys [
         {arriveTangent=9.0725, interpMode=RCIM_Cubic, leaveTangent=9.0725, tangentMode=RCTM_User, time=0, value=0}
         {arriveTangent=0.5892, interpMode=RCIM_Cubic, leaveTangent=0.5892, tangentMode=RCTM_User, time=0.1408, value=0.9847}
         {arriveTangent=-2.1532, interpMode=RCIM_Cubic, leaveTangent=-2.1532, tangentMode=RCTM_User, time=1, value=0}
        ]
       }}})
    ScaleA=true
    ScaleRGB=true
   )
   /Niagara/Modules/Update/Forces/V2/CurlNoiseForce(Bias Noise=false, Noise Strength=&RandomRangeFloat(Minimum=500, Maximum=1250), Pan Noise=false)
   &Drag(Drag=&RandomRangeFloat(Minimum=3, Maximum=6), Use Linear Drag=true, Use Rotational Drag=true)
   &SolveForcesAndVelocity(Clamp Velocity=false, Limit Acceleration=false)
  ]
  rendererClasses [&NiagaraSpriteRendererProperties]
  renderers [{rendererClass=&NiagaraSpriteRendererProperties}]
 }
 NE_Empty {
  emitterUpdateScript [
   &EmitterState()
   &SpawnRate(
    SpawnRate=&FloatFromCurve(
     FloatCurve=DataInterface() {propertyValues {Curve {keys [<at 0.72 1 RCIM_Cubic>,<at 0.82 0 RCIM_Cubic>]}}}
     CurveIndex=$Emitter.NormalizedLoopAge
     Scale Curve=32
    )
    Use Spawn Probability=false
   )
  ]
  particleSpawnScript [
   /Niagara/Modules/Spawn/Initialization/InitializeParticle(
    Lifetime=5
    Module.Write Color=true
    Module.Write Lifetime=true
    Module.Write Mass=true
    Module.Write Material Random=true
    Module.Write Position=false
    Module.Write Scale=false
    Module.Write SpriteRotation=false
    Module.Write SpriteSize=true
    Position=&SimulationPosition()
    Sprite Size=<5 5>
   ) {enabled=false}
   /Niagara/Modules/Spawn/Initialization/V2/InitializeParticle(
    Color Mode=::NewEnumerator0
    Lifetime=0.4
    Mesh Scale=<0.03 0.08 1>
    Mesh Scale Mode=::NewEnumerator4
    Sprite Rotation Angle=&RandomRangeFloat(Minimum=-0.1, Maximum=0.1)
    Sprite Rotation Mode=::NewEnumerator1
    Sprite Size Max=<12 15>
    Sprite Size Min=<9 11>
    Sprite Size Mode=::NewEnumerator2
    Sprite UV Mode=::NewEnumerator3
    UsePositionOffset=false
    Write Material Random=true
   ) {module=InitializeParticle001}
   /SetVariables(Particles.Random size:NiagaraFloat=&RandomRangeFloat(Minimum=0.5, Maximum=0.7)) {isSetParameters=true}
   &AddVelocity(
    Cone Angle=15
    Cone Axis=<1 0 0.5>
    Module.UseConeFalloff=false
    Module.UseVelDistribution=false
    Override Seed=false
    Velocity Mode=::NewEnumerator2
   )
   &CameraOffset(Camera Offset Amount=&RandomRangeFloat(Minimum=2, Maximum=6))
  ]
  particleUpdateScript [
   &ParticleState()
   &SkeletalMeshLocation(
    Bone Sampling Mode=::NewEnumerator4
    MaskSpawnGroup=false
    Orientation Sampling=::NewEnumerator1
    Sampled Position Offset=&AddVectorToPosition(Position=<15 -3 0>, Vector=&RandomRangeVector(Minimum=<-3 -2 -3>, Maximum=<3 2 3>))
    Skeletal Mesh=DataInterface() {propertyValues {
      DefaultMesh=null
      ExcludeBoneName=root
      OptionalComponentClass=null
      PreviewMesh=/BRCosmetics/Characters/Player/Male/Medium/Heads/M_MED_FlickLedge_Head/Meshes/M_MED_FlickLedge_Head
      SkinningMode=SkinOnTheFly
      SoftSourceActor=null
      SourceMode=Default
      UvSetIndex=0
      WholeMeshLOD=-1
      bExcludeBone=false
      bReadDeformedGeometry=true
      bRequireCurrentFrameData=true
      ComponentTags []
      FilteredBones [head]
      FilteredSockets []
      MeshUserParameter {parameter {name=None, typeDefHandle {registeredTypeIndex=1}, varData []}}
      SamplingRegions []
     }}
   )
   &SpriteRotationRate(Rotation Rate=&RandomRangeFloat(Minimum=-200, Maximum=200)) {enabled=false}
   &self:ScratchModule_06() {module=ScratchModule_06}
   &self:ScratchModule_04(UpSpeed:NiagaraFloat=&FloatFromCurve(
     FloatCurve=DataInterface() {propertyValues {Curve {keys [
         <at 0 1 RCIM_Cubic>
         {arriveTangent=0.8911, interpMode=RCIM_Cubic, leaveTangent=0.8911, time=0.1183, value=1.15}
         {arriveTangent=0.0481, interpMode=RCIM_Cubic, leaveTangent=0.0481, tangentMode=RCTM_User, time=0.1683, value=1.15}
         {arriveTangent=0.0179, interpMode=RCIM_Cubic, leaveTangent=0.0179, tangentMode=RCTM_User, time=0.4, value=1.0041}
         {arriveTangent=-0.0074, interpMode=RCIM_Cubic, leaveTangent=-0.0074, tangentMode=RCTM_User, time=0.52, value=0.725}
         {arriveTangent=-0.0755, interpMode=RCIM_Cubic, leaveTangent=-0.0755, tangentMode=RCTM_User, time=0.72, value=0.725}
         {arriveTangent=-8.4445, interpMode=RCIM_Cubic, leaveTangent=-8.4445, tangentMode=RCTM_User, time=0.8, value=0}
        ]}}}
     CurveIndex=$Emitter.NormalizedLoopAge
     Scale Curve=15
    )) {module=ScratchModule_04}
   &self:ScratchModule_05(JogLag=0.05, JogSpeed=300, SkyDiveLag=0.01, SkyDiveSpeed=700) {module=ScratchModule_05}
   &CurlNoiseForce(
    Bias Noise=false
    Noise Frequency=75
    Noise Strength=&Multiply_Float(A=500, B=$Particles.NormalizedAge)
    Pan Noise=true
    Pan Noise Field=<0 0 1>
   )
   &SubUVAnimation(
    End Frame Range Override=3
    Evaluation Type=::NewEnumerator0
    "Sprite Renderer "=DataInterface() {propertyValues {SpriteRenderer=&self:NE_Empty.NiagaraSpriteRendererProperties_1}}
    SubUV Animation Mode:ENiagara_SubUVLookupModeV2=::NewEnumerator2
    UseEndFrame=true
    UseStartFrame=true
   )
   &SolveForcesAndVelocity(Clamp Velocity=false, Limit Acceleration=false)
   &Color(Color=&ColorFromCurve(ColorCurve=DataInterface() {propertyValues {
       AlphaCurve {keys [<at 0 0 RCIM_Cubic>,<at 0.2471 2>,<at 0.5886 2>,<at 1 0>]}
       BlueCurve {keys [<at 0.5879 0.7784>,<at 1 0.1496>]}
       GreenCurve {keys [<at 0.5879 0.4302>,<at 1 0.0827>]}
       RedCurve {keys [<at 0.5879 0>,<at 1 0>]}
      }})) {enabled=false}
   &ScaleSpriteSize(Uniform Curve Sprite Scale=DataInterface() {propertyValues {Curve {keys [
        {arriveTangent=2.652, interpMode=RCIM_Cubic, time=0, value=0.6}
        {arriveTangent=0.688, interpMode=RCIM_Cubic, leaveTangent=0.688, time=0.3, value=1.0816}
        {arriveTangent=-1.2594, interpMode=RCIM_Cubic, leaveTangent=-1.2594, time=0.7, value=1.0816}
        {arriveTangentWeight=0.1011, interpMode=RCIM_Cubic, tangentMode=RCTM_User, time=1, value=0.2}
       ]}}})
   &ScaleColor(
    Scale Alpha=&Multiply_Float(
     A=&FloatFromCurve(FloatCurve=DataInterface() {propertyValues {Curve {keys [
          <at 0 0 RCIM_Cubic>
          {arriveTangent=1.4253, interpMode=RCIM_Cubic, leaveTangent=1.4253, time=0.2063, value=1}
          {arriveTangent=-1.2598, interpMode=RCIM_Cubic, leaveTangent=-1.2598, time=0.7016, value=1}
          <at 1 0 RCIM_Cubic>
         ]}}})
     B=&FloatFromCurve(
      FloatCurve=DataInterface() {propertyValues {Curve {keys [
          {arriveTangent=0.0986, interpMode=RCIM_Cubic, leaveTangent=0.0986, tangentMode=RCTM_User, time=0.4965, value=1}
          {arriveTangent=0.062, interpMode=RCIM_Cubic, leaveTangent=0.062, tangentMode=RCTM_User, time=0.72, value=1}
          {arriveTangent=-21.3048, interpMode=RCIM_Cubic, leaveTangent=-21.3048, tangentMode=RCTM_User, time=0.85, value=0}
         ]}}}
      CurveIndex=$Emitter.NormalizedLoopAge
     )
    )
    Scale RGB=&MakeVectorFromLinearColorRGB(Color=&MultiplyLinearColor(
      Linear Color=&ColorFromCurve(
       ColorCurve=DataInterface() {propertyValues {
         AlphaCurve {keys [<at 0 1>]}
         BlueCurve {keys [<at 0.4159 0 RCIM_Cubic>,<at 0.4692 1 RCIM_Cubic>]}
         GreenCurve {keys [<at 0.4159 1.3201 RCIM_Cubic>,<at 0.4692 0.5526 RCIM_Cubic>]}
         RedCurve {keys [<at 0.4159 4 RCIM_Cubic>,<at 0.4692 0 RCIM_Cubic>]}
        }}
       CurveIndex=$Emitter.NormalizedLoopAge
      )
      Scale Factor=&MakeVector4_FromFloat(Float=&FloatFromCurve(FloatCurve=DataInterface() {propertyValues {Curve {keys [<at 0.6 0.776>,<at 1 0.5>]}}}))
     ))
    ScaleA=true
    ScaleRGB=true
   )
   /SetVariables(
    Particles.CoreInFrontColor:LinearColor=&Multiply_LinearColorByFloat(
     Linear Color=&MultiplyLinearColor(
      Linear Color=$Particles.Color
      Scale Factor=&MakeVector4(
       X=1
       Y=1
       Z=1
       W=&ClampFloat(Float=&RemapRange(
         Input Value To Remap=&DotProduct(
          Vector A=&NormalizeVector(
           Vector To Normalize=&ReturnCameraVector(Camera Vectors=::NewEnumerator1, Coordinate Space=::Simulation)
           Safe Normalize=true
           Safe Fallback Threshold=0
           Safe Fallback Vector=<0 0 0.001>
          )
          Vector B=&RotateVectorByQuaternion(Vector To Rotate=<-1 0 0>, Quaternion=$Particles.SkeletalMeshLocation.SampledRotation)
         )
         Input Min=0.3
         Input Max=0.6
         Output Min=0
         Output Max=1
        ))
      )
     )
     Float=1
    )
    Particles.GlowColor=&MultiplyLinearColor(Linear Color=$Particles.Color, Scale Factor=<1 1 0.6 0.15>)
    Particles.GlowScale:Vector2f=&Multiply_Vector2D_ByFloat(Float=1.4, Vector2D=$Particles.SpriteSize)
   ) {isSetParameters=true}
  ]
  rendererClasses [&NiagaraSpriteRendererProperties]
  renderers [
   {rendererClass=&NiagaraSpriteRendererProperties}
   {rendererClass=&NiagaraSpriteRendererProperties, rendererIndex=1}
   {rendererClass=&NiagaraSpriteRendererProperties, rendererIndex=2}
   {rendererClass=&NiagaraSpriteRendererProperties, rendererIndex=3}
  ]
 }
 SmokeWorld {
  emitterUpdateScript [
   &EmitterState()
   &SpawnRate(
    SpawnRate=&FloatFromCurve(
     FloatCurve=DataInterface() {propertyValues {Curve {keys [<at 0 1>,<at 1.05 1>,<at 1.083 0.85>,<at 2 0.85>,<at 2.033 1>,<at 2.6426 1>,<at 3.1941 1.0194>,<at 3.25 0>]}}}
     CurveIndex=$Emitter.LoopedAge
     Scale Curve=24
    )
    Use Spawn Probability=false
   )
  ]
  particleSpawnScript [
   /Niagara/Modules/Spawn/Initialization/InitializeParticle(
    Color=&UniformRangedLinearColor(Minimum=<rgba 0.0156 0.0148 0.0133 0.125>, Maximum=<rgba 0 0 0 0.425>)
    Lifetime=&RandomRangeFloat(Minimum=1.2, Maximum=1.6)
    Module.Write Color=true
    Module.Write Lifetime=true
    Module.Write Mass=true
    Module.Write Material Random=true
    Module.Write Position=false
    Module.Write Scale=false
    Module.Write SpriteRotation=true
    Module.Write SpriteSize=true
    Position=&SimulationPosition()
    Sprite Rotation=&UniformRangedFloat(Minimum=0, Maximum=360)
    Sprite Size=&Vector2DFromFloat(Value=&UniformRangedFloat(Minimum=80, Maximum=100))
   )
   &ShapeLocation(MaskSpawnGroup=false, Override Seed=false, Sphere Radius=12, UseOwnerScale=true)
   &SkeletalMeshLocation(
    MaskSpawnGroup=false
    Mesh Sampling Type=::NewEnumerator5
    Sampled Position Offset:NiagaraPosition=$Particles.ShapeLocation.ShapePosition
    Skeletal Mesh=$System.Skeletal Mesh Reader
    Socket Sampling Mode=::NewEnumerator5
   )
   &AddVelocity(Velocity=&MakeVector(
     X=0
     Y=0
     Z=&FloatFromCurve(
      FloatCurve=DataInterface() {propertyValues {Curve {keys [<at 0 1>,<at 1.05 1>,<at 1.083 0.4>,<at 2 0.4>,<at 2.033 1.25>,<at 2.6426 1.25>,<at 3.1941 1.25>,<at 3.5 0.45>,<at 5.75 0.45>,<at 6 0>]}}}
      CurveIndex=$Emitter.LoopedAge
      Scale Curve=&RandomRangeFloat(Minimum=245, Maximum=325)
     )
    ))
  ]
  particleUpdateScript [
   &ParticleState()
   &SpriteRotationRate(Rotation Rate=&RandomRangeFloat(Minimum=-20, Maximum=20)) {enabled=false}
   &Drag(Drag=&RandomRangeFloat(Minimum=4, Maximum=6), Use Linear Drag=true, Use Rotational Drag=false)
   &SolveForcesAndVelocity(Clamp Velocity=false, Limit Acceleration=false)
   &ScaleSpriteSize(Scale Factor=&Vector2DFromFloat(Value=&FloatFromCurve(FloatCurve=DataInterface() {propertyValues {Curve {keys [
          {arriveTangent=2.5043, interpMode=RCIM_Cubic, leaveTangent=2.5043, tangentMode=RCTM_User, time=0, value=0.2}
          {arriveTangent=0.2677, interpMode=RCIM_Cubic, leaveTangent=0.2677, tangentMode=RCTM_User, time=1, value=1}
         ]}}})))
   &DynamicMaterialParameters(
    Index 0 Param 1=&FloatFromCurve(FloatCurve=DataInterface() {propertyValues {Curve {keys [<at -0.0046 1.0113 RCIM_Cubic>,<at 1.0011 0.002>]}}})
    Index 0 Param 4=&FloatFromCurve(FloatCurve=DataInterface() {propertyValues {Curve {keys [<at 0 0 RCIM_Cubic>,<at 1 0.5>]}}})
    Module.Param0WriteEnabled=true
    Module.Param1WriteEnabled=false
    Module.Param2WriteEnabled=false
    Module.Param3WriteEnabled=true
   ) {module=DynamicMatParams}
   &CameraOffset(Camera Offset Amount=50)
   &ScaleColor(
    Scale Alpha=&Multiply_Float(
     A=&FloatFromCurve(
      FloatCurve=DataInterface() {propertyValues {Curve {keys [
          {arriveTangent=45.1311, interpMode=RCIM_Cubic, leaveTangent=45.1311, tangentMode=RCTM_User, time=0, value=0}
          {arriveTangent=0.002, interpMode=RCIM_Cubic, leaveTangent=0.002, time=0.05, value=1}
          <at 1.0011 0.002>
         ]}}}
      Scale Curve=0.4
     )
     B=&FloatFromCurve(
      FloatCurve=DataInterface() {propertyValues {Curve {defaultValue=1, keys [<at 0 1.4>,<at 1.05 1.4>,<at 1.083 1>,<at 2 1>,<at 2.033 1.4>,<at 2.6426 1.4>,<at 3.1941 0.4>,<at 5.75 0.4>,<at 6 0>]}}}
      CurveIndex=$Emitter.LoopedAge
      Scale Curve=1.7
     )
    )
    ScaleA=true
    ScaleRGB=false
   )
   &SubUVAnimation(
    "Sprite Renderer "=DataInterface() {propertyValues {SpriteRenderer=&self:SmokeWorld_1.NiagaraSpriteRendererProperties_1}}
    UseEndFrame=false
    UseStartFrame=false
   )
  ]
  rendererClasses [&NiagaraSpriteRendererProperties]
  renderers [
   {rendererClass=&NiagaraSpriteRendererProperties}
   {rendererClass=&NiagaraSpriteRendererProperties, rendererIndex=1}
  ]
 }
}
summary {
 systemName=NS_Emote_FlickLedge_Fire
 emitters [
  {emitterName=SmokeWorld, rendererClasses [&NiagaraSpriteRendererProperties]}
  {emitterName=Flames003, rendererClasses [&NiagaraSpriteRendererProperties]}
  {emitterName=NE_Embers001, rendererClasses [&NiagaraSpriteRendererProperties]}
  {emitterName=Flames004, rendererClasses [&NiagaraSpriteRendererProperties]}
  {emitterName=Flames005, rendererClasses [&NiagaraSpriteRendererProperties, &NiagaraLightRendererProperties]}
  {emitterName=BurstFlame, rendererClasses [&NiagaraSpriteRendererProperties]}
  {emitterName=NE_Empty, rendererClasses [&NiagaraSpriteRendererProperties]}
  {emitterName=Empty, rendererClasses []}
 ]
 userVariables [
  User.bIsFrontEnd {type=NiagaraBool}
  User.bIsFrontEndPreview {type=NiagaraBool}
  User.bisSwinging {type=NiagaraBool}
  User.Emote_DuckCoast_FireVelX {type=NiagaraFloat, defaultValue=0, description=""}
  User.Emote_DuckCoast_FireVelY {type=NiagaraFloat, defaultValue=0, description=""}
  User.Emote_DuckCoast_FireVelZ {type=NiagaraFloat, defaultValue=0, description=""}
  User.Emote_MatchaSpare_SizeX {type=NiagaraFloat, defaultValue=0, description=""}
  User.Emote_MatchaSpare_SizeY {type=NiagaraFloat, defaultValue=0, description=""}
  User.Speed {type=NiagaraFloat, defaultValue=0, description=""}
  User.VariantID {type=NiagaraFloat, defaultValue=0, description=""}
 ]
}
